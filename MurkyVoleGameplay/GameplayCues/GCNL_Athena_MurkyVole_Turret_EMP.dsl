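; Looping gameplay cue that shows the MurkyVole turret EMP effect on a Fortnite player pawn's backpack.
; On application it casts the target to a player pawn and, if no Niagara component is stored, spawns the EMP system on the backpack mesh's turret_rail socket.
; Removal deactivates the stored component.
; Helper functions cache the pawn and backpack mesh, but neither event here calls them.
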(event EventOnApplication(EitherVFX-preferthisone)
  (params (MyTarget :type Actor%20Object%20Reference) (Parameters :type Gameplay%20Cue%20Parameters%20Structure%20%28by%20ref%29) (ParticleComponents :type Array%20of%20FXSystem%20Component%20Object%20References) (AudioComponents :type Array%20of%20Audio%20Component%20Object%20References) (BurstCameraShakeInstance :type Camera%20Shake%20Base%20Object%20Reference) (BurstDecalInstance :type Decal%20Actor%20Object%20Reference))
  (bind _asfort_player_pawn (Utilities|Casting|CastToFortPlayerPawn MyTarget)
    (:then
      (Variables|Default|GetNiagara%20System
        (:then)
        (:else
          (bind _returnvalue (Niagara|SpawnSystemAttached "/MurkyVoleGameplay/FX/Systems/NS_MurkyVole_TurretEMP.NS_MurkyVole_TurretEMP" (Customization|GetSkeletalMeshforPartType _asfort_player_pawn "Backpack" :!class /Script/FortniteGame.FortPlayerPawn) "turret_rail" "0, 0, 0" "0, 0, 0" "KeepRelativeOffset" true))
          (Variables|Default|SetNiagara%20System _returnvalue))))
    (:CastFailed
      (Development|PrintString 0 "cast Failed"))))

(event EventOnRemoval(EitherVFX-preferthisversion)
  (params (MyTarget :type Actor%20Object%20Reference) (Parameters :type Gameplay%20Cue%20Parameters%20Structure%20%28by%20ref%29) (ParticleComponents :type Array%20of%20FXSystem%20Component%20Object%20References) (AudioComponents :type Array%20of%20Audio%20Component%20Object%20References) (BurstCameraShakeInstance :type Camera%20Shake%20Base%20Object%20Reference) (BurstDecalInstance :type Decal%20Actor%20Object%20Reference))
  (Components|Activation|Deactivate (Variables|Default|GetNiagara%20System)))

(fn CachePawnAndBackpack
  (params (TargetActor :type Object%20Reference))
  (bind _asfort_player_pawn (Utilities|Casting|CastToFortPlayerPawn TargetActor))
  (bind _output_get (Variables|Default|SetTargetPawn _asfort_player_pawn))
  (Variables|Default|SetSkeletalMesh_BackPack (Customization|GetSkeletalMeshforPartType _output_get "Backpack" :!class /Script/FortniteGame.FortPlayerPawn)))

(fn TryGetBackpackMesh
  (bind _skeletalmesh_backpack (Variables|Default|GetSkeletalMesh_BackPack)
    (:then
      (return _skeletalmesh_backpack))
    (:else
      (Utilities|FlowControl|Sequence
        (:then_0
          (bind _targetpawn (Variables|Default|GetTargetPawn)
            (:then
              (Utilities|IsValid (Customization|GetSkeletalMeshforPartType _targetpawn "Backpack" :!class /Script/FortniteGame.FortPlayerPawn)
                (:Is%20Valid
                  (Variables|Default|SetSkeletalMesh_BackPack (Customization|GetSkeletalMeshforPartType _targetpawn "Backpack" :!class /Script/FortniteGame.FortPlayerPawn)))))
            (:else)))
        (:then_1
          (return (Variables|Default|GetSkeletalMesh_BackPack)))))))

(fn ConstructionScript)
