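; Burst gameplay cue for the MurkyVole missile swarm explosion, adding sound, camera shake and a sound indicator after the parent burst logic.
; It shows a gunshot-type sound indicator at the target, ranged by the non-owner sound's max distance.
; If the effect causer's instigator is locally controlled, owner shake and sound variants play; otherwise non-owner ones with a tighter shake radius.
; Shakes center on the target, sounds on the cue location.
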(fn OnBurst(EitherVFX-preferthisversion)
  (params (MyTarget :type Actor%20Object%20Reference) (Parameters :type Gameplay%20Cue%20Parameters%20Structure%20%28by%20ref%29) (ParticleComponents :type Array%20of%20FXSystem%20Component%20Object%20References) (AudioComponents :type Array%20of%20Audio%20Component%20Object%20References) (BurstCameraShakeInstance :type Camera%20Shake%20Base%20Object%20Reference) (BurstDecalInstance :type Decal%20Actor%20Object%20Reference))
  (bind _returnvalue (Transformation|GetActorLocation MyTarget))
  (bind (_normalizedmagnitude _rawmagnitude _effectcontext _matchedtagname _originaltag _aggregatedsourcetags _aggregatedtargettags _location _normal _instigator _effectcauser) (Ability|GameplayCue|BreakGameplayCueParameters Parameters))
  (bind (_normalizedmagnitude_2 _rawmagnitude_2 _effectcontext_2 _matchedtagname_2 _originaltag_2 _aggregatedsourcetags_2 _aggregatedtargettags_2 _location_2 _normal_2 _instigator_2 _effectcauser_2) (Ability|GameplayCue|BreakGameplayCueParameters (Variables|GetParameters)))
  (bind (_normalizedmagnitude_3 _rawmagnitude_3 _effectcontext_3 _matchedtagname_3 _originaltag_3 _aggregatedsourcetags_3 _aggregatedtargettags_3 _location_3 _normal_3 _instigator_3 _effectcauser_3) (Ability|GameplayCue|BreakGameplayCueParameters (Variables|GetParameters)))
  (|CallParentFunction MyTarget Parameters ParticleComponents AudioComponents BurstCameraShakeInstance BurstDecalInstance)
  (Fortnite|ShowSoundIndicatorLocal :WorldContextObject MyTarget :TrackedActor MyTarget :Location _returnvalue :MaxAudibleDistance (Variables|/Script/Engine.SoundBase|GetMaxDistance (Variables|Default|GetSound_Explosion_NonOwner)) :IndicatorType "Gunshot" :Instigator MyTarget :Tint "(R=0.000000,G=0.000000,B=0.000000,A=0.000000)" :AttenuationMaxRangeMultiplier 1.0 :AttenuationIntensityMultiplier 1.0)
  (if (Pawn|IsLocallyControlled (Variables|/Script/Engine.Actor|GetInstigator _effectcauser))
    (Camera|PlayWorldCameraShake MyTarget (Variables|Default|GetCameraShake_Owner) _returnvalue 384.0 2.56e+03)
    (Audio|PlaySoundatLocation _effectcauser_3 (Variables|Default|GetSound_Explosion_Owner) _location_3)
    (else
      (Camera|PlayWorldCameraShake MyTarget (Variables|Default|GetCameraShake_NonOwner) _returnvalue 0.0 1792.0)
      (Audio|PlaySoundatLocation _effectcauser_2 (Variables|Default|GetSound_Explosion_NonOwner) _location_2)))
  (return))
