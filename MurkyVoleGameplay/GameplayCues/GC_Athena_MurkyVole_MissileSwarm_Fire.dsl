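; Burst gameplay cue for firing the MurkyVole missile swarm, playing muzzle VFX and shot audio on the firing player pawn.
; Raw magnitude carries the missile number, which picks a muzzle socket by modulo, cycling the Niagara flash across the backpack's sockets.
; The locally controlled shooter gets 2D first-person audio; others get third-person sound at the pawn plus a gunshot sound indicator.
; Both sounds receive a "FireSingleShot" trigger.
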(fn OnBurst(EitherVFX-preferthisversion)
  (params (MyTarget :type Actor%20Object%20Reference) (Parameters :type Gameplay%20Cue%20Parameters%20Structure%20%28by%20ref%29) (ParticleComponents :type Array%20of%20FXSystem%20Component%20Object%20References) (AudioComponents :type Array%20of%20Audio%20Component%20Object%20References) (BurstCameraShakeInstance :type Camera%20Shake%20Base%20Object%20Reference) (BurstDecalInstance :type Decal%20Actor%20Object%20Reference))
  (bind (_normalizedmagnitude _rawmagnitude) (Ability|GameplayCue|BreakGameplayCueParameters Parameters))
  (|CallParentFunction MyTarget Parameters ParticleComponents AudioComponents BurstCameraShakeInstance BurstDecalInstance)
  (LocalVariables|SetMissileNumber (Math|Float|Truncate _rawmagnitude))
  (bind _asfort_player_pawn (Utilities|Casting|CastToFortPlayerPawn MyTarget)
    (:then
      ;[1 1 1] VFX
      (LocalVariables|SetPlayerPawn _asfort_player_pawn)
      (bind _asccpm_athena_murky_vole_back_pack (Utilities|Casting|CastToCCPM_Athena_MurkyVole_BackPack (FortPawn|GetModifierforPartType _asfort_player_pawn "Backpack"))
        (:then
          (Niagara|SpawnSystemAttached (Variables|Default|GetNS_Muzzle) (Class|CCPMAthenaMurkyVoleBackPack|GetSocketLocationandRotation _asccpm_athena_murky_vole_back_pack "None") (Utilities|Array|Get(acopy) (Variables|Default|GetMuzzleSocketNames) (Math|Integer|%25%28Integer%29 (- (LocalVariables|GetMissileNumber) 1) (Utilities|Array|Length (Variables|Default|GetMuzzleSocketNames)))) "0, 0, 0" "0, 0, 0" "SnapToTarget" true true "AutoRelease")
          ;[1 1 1] Sound
          (if (Pawn|IsLocallyControlled (LocalVariables|GetPlayerPawn))
            (bind _returnvalue (Audio|SpawnSound2D (LocalVariables|GetPlayerPawn) (Variables|Default|GetSound_1P)))
            (Audio|Parameter|ExecuteTriggerParameter _returnvalue "FireSingleShot")
            (else
              (Fortnite|ShowSoundIndicatorLocal :WorldContextObject (LocalVariables|GetPlayerPawn) :TrackedActor (LocalVariables|GetPlayerPawn) :Location (Transformation|GetActorLocation (LocalVariables|GetPlayerPawn)) :MaxAudibleDistance (ScalableFloat|GetValueatLevel (Variables|Default|GetRow_AVIndicatorDistance)) :IndicatorType "Gunshot" :Instigator (LocalVariables|GetPlayerPawn) :Tint "(R=0.000000,G=0.000000,B=0.000000,A=0.000000)" :AttenuationMaxRangeMultiplier 1.0 :AttenuationIntensityMultiplier 1.0)
              (bind _returnvalue_2 (Audio|SpawnSoundatLocation (LocalVariables|GetPlayerPawn) (Variables|Default|GetSound_3P) (Transformation|GetActorLocation (LocalVariables|GetPlayerPawn))))
              (Audio|Parameter|ExecuteTriggerParameter _returnvalue_2 "FireSingleShot")))
          (return))
        (:CastFailed
          (return))))
    (:CastFailed
      (return))))
