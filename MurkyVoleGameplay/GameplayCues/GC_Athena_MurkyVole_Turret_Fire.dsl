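; Gameplay cue burst for MurkyVole backpack turret shots, playing fire and impact feedback and sending a ShotFired event to the player pawn.
; Fire audio triggers reused sound-library components with 1P/3P variants for performance.
; The backpack socket anchors muzzle flash and beam, and non-local shooters show a gunshot indicator.
; Valid hit actors get physical-material FX with flesh fallback, flesh-or-stone sound, and a bounce on damageable buildings.
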(fn OnBurst(EitherVFX-preferthisversion)
  (params (MyTarget :type Actor%20Object%20Reference) (Parameters :type Gameplay%20Cue%20Parameters%20Structure%20%28by%20ref%29) (ParticleComponents :type Array%20of%20FXSystem%20Component%20Object%20References) (AudioComponents :type Array%20of%20Audio%20Component%20Object%20References) (BurstCameraShakeInstance :type Camera%20Shake%20Base%20Object%20Reference) (BurstDecalInstance :type Decal%20Actor%20Object%20Reference))
  (bind (_value _returnvalue) (Utilities|Map|Find (Variables|Default|GetMAP_PhysMaterial_ImpactFX) (LocalVariables|GetPhysMat)))
  (bind (_normalizedmagnitude _rawmagnitude _effectcontext _matchedtagname _originaltag _aggregatedsourcetags _aggregatedtargettags _location _normal _instigator _effectcauser _sourceobject _physicalmaterial) (Ability|GameplayCue|BreakGameplayCueParameters Parameters))
  (bind _muzzlesocket (LocalVariables|GetMuzzleSocket))
  (|CallParentFunction MyTarget Parameters ParticleComponents AudioComponents BurstCameraShakeInstance BurstDecalInstance)
  (bind _output_get (LocalVariables|SetPlayer MyTarget))
  (bind _asfort_player_pawn (Utilities|Casting|CastToFortPlayerPawn _output_get)
    (:then
      ; Fire sound logic - takes advantage of the weapon sound library component and reuses spawned audio components instead, improving perf
      (bind _asbp_athena_murky_vole_turret_attachment (Utilities|Casting|CastToBP_Athena_MurkyVole_Turret_Attachment _sourceobject)
        (:then
          (Utilities|IsValid (Variables|/MurkyVoleGameplay/Gameplay/Turret/BP_Athena_MurkyVole_Turret_Attachment.BP_Athena_MurkyVole_Turret_Attachment_C|GetWeaponSoundLibraryComponent _asbp_athena_murky_vole_turret_attachment)
            (:Is%20Valid
              (bind _returnvalue_2 (SoundLibrary|PlayEvent (Variables|/MurkyVoleGameplay/Gameplay/Turret/BP_Athena_MurkyVole_Turret_Attachment.BP_Athena_MurkyVole_Turret_Attachment_C|GetWeaponSoundLibraryComponent _asbp_athena_murky_vole_turret_attachment) (select (Pawn|IsLocallyViewed _asfort_player_pawn) "(TagName=\"SoundLibrary.Weapon.Fire.1P\")" "(TagName=\"SoundLibrary.Weapon.Fire.3P\")")))
              (bind (_returnvalue_eventname _returnvalue_bwasblocked _returnvalue_bsuccess _returnvalue_audiocomponents) (Utilities|Struct|BreakSoundLibraryPlayResult _returnvalue_2))
              (Utilities|IsValid (Utilities|Array|Get(acopy) _returnvalue_audiocomponents)
                (:Is%20Valid
                  (Audio|Parameter|ExecuteTriggerParameter (Utilities|Array|Get(acopy) _returnvalue_audiocomponents) "FireSingleShot"))))
            (:Is%20Not%20Valid
              _))
          (bind _asccpm_athena_murky_vole_back_pack (Utilities|Casting|CastToCCPM_Athena_MurkyVole_BackPack (FortPawn|GetModifierforPartType _asfort_player_pawn "Backpack"))
            (:then
              (LocalVariables|SetTurretMeshComponent (Class|CCPMAthenaMurkyVoleBackPack|GetSocketLocationandRotation _asccpm_athena_murky_vole_back_pack _muzzlesocket))
              (Ability|SendGameplayEventtoActor _asfort_player_pawn "(TagName=\"Event.Athena.MurkyVole.Turret.ShotFired\")")
              (LocalVariables|SetMuzzleLocation (.SocketLocation (Class|CCPMAthenaMurkyVoleBackPack|GetSocketLocationandRotation _asccpm_athena_murky_vole_back_pack _muzzlesocket)))
              (LocalVariables|SetHitLocation _location)
              (LocalVariables|SetHitNormal _normal)
              (LocalVariables|SetHitActor _effectcauser)
              (LocalVariables|SetPhysMat _physicalmaterial)
              (bind _returnvalue_3 (Effects|SpawnFXSystemAttached (Variables|Default|GetNS_MuzzleFlash) (LocalVariables|GetTurretMeshComponent) (LocalVariables|GetMuzzleSocket)))
              (Effects|Components|ParticleSystem|SetVectorParameter _returnvalue_3 "BeamEnd_0" (LocalVariables|GetHitLocation))
              (if (not (Pawn|IsLocallyControlled (Variables|/Script/Engine.Actor|GetInstigator (LocalVariables|GetPlayer))))
                (Fortnite|ShowSoundIndicatorLocal :WorldContextObject (LocalVariables|GetPlayer) :TrackedActor (LocalVariables|GetPlayer) :Location (LocalVariables|GetMuzzleLocation) :MaxAudibleDistance (ScalableFloat|GetValueatLevel (Variables|Default|GetRow_AV_IndicatorRange)) :IndicatorType "Gunshot" :Instigator (LocalVariables|GetPlayer) :Tint "(R=0.000000,G=0.000000,B=0.000000,A=0.000000)" :SoundBase (Variables|Default|GetSound_Fire_Remote) :AttenuationMaxRangeMultiplier 1.0 :AttenuationIntensityMultiplier 1.0))
              (Utilities|IsValid (LocalVariables|GetHitActor)
                (:Is%20Valid
                  (Effects|Components|ParticleSystem|SpawnFXAtLocation (LocalVariables|GetPlayer) (select _returnvalue _value (Utilities|Map|Find (Variables|Default|GetMAP_PhysMaterial_ImpactFX) "/Game/FortressPhysicalMaterials/Player_Flesh.Player_Flesh")) (LocalVariables|GetHitLocation) (Math|Conversions|RotationFromXVector (Math|Vector|SelectVector (- (LocalVariables|GetMuzzleLocation) (LocalVariables|GetHitLocation)) (LocalVariables|GetHitNormal) (Utilities|Equal(Object) (LocalVariables|GetPhysMat) "/Game/FortressPhysicalMaterials/Player_Flesh.Player_Flesh"))) "1.000000,1.000000,1.000000" true "AutoRelease")
                  (Audio|PlaySoundatLocation (LocalVariables|GetPlayer) (select (== (LocalVariables|GetPhysMat) "/Game/FortressPhysicalMaterials/Player_Flesh.Player_Flesh") "/Game/Sounds/Fort_Impact_Sounds/Flesh/Player_Impact_Flesh_Spatial_Cue.Player_Impact_Flesh_Spatial_Cue" "/Game/Sounds/Fort_Weapons/Impacts/STONE_Bullet_Impact_Ricochet_Heavy_Cue.STONE_Bullet_Impact_Ricochet_Heavy_Cue") (LocalVariables|GetHitLocation)) ; Removing map selection in favor of perf.
                  (bind _asbuilding_static_mesh (Utilities|Casting|CastToBuildingSMActor (LocalVariables|GetHitActor))
                    (:then
                      (if (Variables|/Script/Engine.Actor|GetbCanBeDamaged _asbuilding_static_mesh)
                        (Class|BuildingStaticMesh|StartBounceAnimation _asbuilding_static_mesh (Utilities|Time|GetTimeSeconds (LocalVariables|GetPlayer)) 2e+02 (LocalVariables|GetHitNormal) (LocalVariables|GetHitLocation) "Hit" true)
                        (return)
                        (else
                          (return))))
                    (:CastFailed
                      (return))))
                (:Is%20Not%20Valid
                  (return))))
            (:CastFailed
              (return))))
        (:CastFailed
          _)))
    (:CastFailed
      (return))))
